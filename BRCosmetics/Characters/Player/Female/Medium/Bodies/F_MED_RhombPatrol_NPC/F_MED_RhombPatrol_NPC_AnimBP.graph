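# Post-process-style Anim Blueprint for the female medium RhombPatrol NPC body part, built on the Rebirth Soldier skeleton.
# It copies the attached parent's pose, offsets glute bones by thigh rotation alphas, then runs bandana and jacket RigidBody simulations and a Control Rig.
# Each update feeds the owner's head-bone axes and offset position into material slot 1 as plane parameters for the visor pattern.
# Two extra glute and jacket nodes are unconnected.
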AnimBlueprint /BRCosmetics/Characters/Player/Female/Medium/Bodies/F_MED_RhombPatrol_NPC/F_MED_RhombPatrol_NPC_AnimBP
target_skeleton=/Game/Characters/Player/Female/Medium/Bodies/F_MED_Rebirth_Soldier/Meshes/F_MED_Rebirth_Soldier_Skeleton
graphs [
 AnimGraph {nodes [
   "Transform (Modify) Bone - Bone: dyn_glute_r"(Alpha=1, Translation="3.750,0,0", Rotation="0.000,0.000,-40", Scale="1.25,1.25,1.25") {
    rotationMode=BMM_Additive
    rotationSpace=BCS_BoneSpace
    scaleMode=BMM_Additive
    scaleSpace=BCS_BoneSpace
    translationMode=BMM_Additive
    translationSpace=BCS_BoneSpace
    alphaBoolBlend {blendOption=Linear}
    boneToModify {boneName=dyn_glute_r}
    rotation {roll=-40}
    scale=<1.25 1.25 1.25>
    translation {x=3.75}
   }
   RigidBody() {
    bClampLinearTranslationLimitToRefPose=true
    bOverrideWorldGravity=true
    overridePhysicsAsset=/Game/Characters/Player/Male/Medium/Bodies/M_MED_Rebirth_Soldier/Meshes/M_MED_Rebirth_Soldier_Dyn_Jacket
    simulationSpace=BaseBoneSpace
    alphaBoolBlend {blendOption=Linear}
    baseBoneRef {boneName=spine_01}
    componentAppliedLinearAccClamp=<3000 3000 3000>
    componentLinearAccScale=<0.5 0.2 0.2>
    componentLinearVelScale=<0.5 0.2 0.2>
    overrideWorldGravity {z=-980}
   }
   n1 = "Transform (Modify) Bone - Bone: dyn_glute_l"(
    ComponentPose=Local To Component(LocalPose=Copy Pose From Mesh() {bUseAttachedParent=true})
    Alpha=Get ThighLeftRotationAlpha() -> (ThighLeftRotationAlpha:float)
    Translation="-3.750,0.875,-0.9"
    Rotation="0.000,0.000,-55"
    Scale="1.25,1.25,1.25"
   ) {
    rotationMode=BMM_Additive
    rotationSpace=BCS_BoneSpace
    scaleMode=BMM_Additive
    scaleSpace=BCS_BoneSpace
    translationMode=BMM_Additive
    translationSpace=BCS_BoneSpace
    alphaBoolBlend {blendOption=Linear}
    boneToModify {boneName=dyn_glute_l}
    rotation {roll=-55}
    scale=<1.25 1.25 1.25>
    translation=<-3.75 0.875 -0.9>
   }
   n2 = "Transform (Modify) Bone - Bone: dyn_glute_r"(
    ComponentPose=@n1
    Alpha=Get ThighRightRotationAlpha() -> (ThighRightRotationAlpha:float)
    Translation="-3.75,0,0"
    Rotation="0.000,0.000,-55"
    Scale="1.3,1.3,1.3"
   ) {
    rotationMode=BMM_Additive
    rotationSpace=BCS_BoneSpace
    scaleMode=BMM_Additive
    scaleSpace=BCS_BoneSpace
    translationMode=BMM_Additive
    translationSpace=BCS_BoneSpace
    alphaBoolBlend {blendOption=Linear}
    boneToModify {boneName=dyn_glute_r}
    rotation {roll=-55}
    scale=<1.3 1.3 1.3>
    translation {x=-3.75}
   }
   n3 = RigidBody(ComponentPose=@n2) {
    bClampLinearTranslationLimitToRefPose=true
    bOverrideWorldGravity=true
    overridePhysicsAsset=/Game/Characters/Player/Female/Medium/Bodies/F_MED_Rebirth_Soldier/Meshes/F_MED_Rebirth_Soldier_Bandana_Physics
    simulationSpace=BaseBoneSpace
    alphaBoolBlend {blendOption=Linear}
    baseBoneRef {boneName=spine_02}
    componentAppliedLinearAccClamp=<3000 3000 3000>
    componentLinearAccScale=<0.5 0.2 0.125>
    componentLinearVelScale=<0.5 0.2 0.125>
    overrideWorldGravity {z=-980}
   }
   n4 = RigidBody(ComponentPose=@n3) {
    bClampLinearTranslationLimitToRefPose=true
    bOverrideWorldGravity=true
    overridePhysicsAsset=/Game/Characters/Player/Female/Medium/Bodies/F_MED_Rebirth_Soldier/Meshes/F_MED_Rebirth_Soldier_Jacket
    simulationSpace=BaseBoneSpace
    alphaBoolBlend {blendOption=Linear}
    baseBoneRef {boneName=spine_02}
    componentAppliedLinearAccClamp=<3000 3000 2800>
    componentLinearAccScale=<0.4 0.4 0.2>
    componentLinearVelScale=<0.4 0.4 0.2>
    overrideWorldGravity {z=-980}
   }
   n5 = Control Rig(Source=Component To Local(ComponentPose=@n4)) {
    alphaBoolBlend {blendOption=Linear}
    controlRigAssetReference {blueprintRigClass=/Game/Characters/Player/Female/Medium/Bodies/F_MED_Rebirth_Soldier/Meshes/F_MED_Rebirth_Soldier_Fresh_CtrlRig.F_MED_Rebirth_Soldier_Fresh_CtrlRig_C}
   }
   Output Pose(Result=@n5) {layerGroup=None}
  ]}
 EventGraph {nodes [
   Event BlueprintUpdateAnimation()
   > n1 = Get Dynamic Material Instance(SkeletalMeshComponent=GetPartSkeletalMeshComponent(), ElementIndex=1)
   n2 = BreakTransform(InTransform=GetBoneTransform!(
     self:Skinned Mesh Component Object Reference=GetOwnerSkeletalMeshComponent()
     InBoneName:Name=head
     TransformSpace:ERelativeTransformSpace Enum=RTS_World
    ) -> (ReturnValue:Transform))
   n3 = GetRightVector!(InRot:Rotator=@n2.Rotation) -> (ReturnValue:Vector)
   > SetVectorParameterValue(
    self=@n1
    ParameterName=VP_PlaneNormal
    Value="To LinearColor (Vector)"!(InVec:Vector=vector * vector(A=@n3, B:Vector="-1,-1,-1") -> (ReturnValue=~)) -> (ReturnValue:Linear Color Structure)
   )
   n4 = GetUpVector!(InRot:Rotator=@n2.Rotation) -> (ReturnValue:Vector)
   > SetVectorParameterValue(self=@n1, ParameterName=VP_XAxis, Value="To LinearColor (Vector)"(InVec=vector * vector(A=@n4, B="-1,-1,-1") -> (ReturnValue=~)))
   n5 = BreakVector(InVec=Get VisorPatternOffset() -> (VisorPatternOffset:Vector))
   n6 = vector * vector(A=@n3, B:double=@n5) -> (ReturnValue=~)
   n7 = vector * vector(A=@n4, B=@n5.Y) -> (ReturnValue=~)
   n8 = vector * vector(A=GetForwardVector!(InRot:Rotator=@n2.Rotation) -> (ReturnValue:Vector), B=@n5.Z) -> (ReturnValue=~)
   > SetVectorParameterValue(
    self=@n1
    ParameterName=VP_PlaneCenter
    Value="To LinearColor (Vector)"(InVec=vector + vector!(A=@n2, B=@n6, C:Vector=@n7, D:Vector=@n8) -> (ReturnValue))
   )
  ]}
]
variables [VisorPatternOffset {type=Vector, default=<9.7 1 5>}]
